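# First-person fire sound cue for the Precision AR, mixing three layers under sound class overrides.
# The shot layer picks one of six close shots, with volume modulated down by TimeSinceStarted.
# A quieter layer adds one of six random shell drops with varied volume.
# A LowAmmo branch adds a warning layer in the critical low-ammo sound class, growing louder as AmmoNormalized falls through the last quarter.
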SoundCue /CorruptionItems/Sounds/Weapons/AR/PrecisionAR/AR_Precision_Fire_1P_Cue
root=Mixer_0
nodes [
 {id=Branch_3, children [ModulatorContinuous_2, null, null], properties {boolParameterName=LowAmmo}}
 {id=Mixer_0, children [SoundClass_0,SoundClass_1,SoundClass_2], properties {inputVolume [1,1,0.3]}}
 {id=Mixer_1, children [null, ModulatorContinuous_0], properties {inputVolume [1,1]}}
 {
  id=ModulatorContinuous_0
  children [Modulator_0]
  properties {volumeModulationParams {default=0, maxOutput=0.6, minOutput=1, parameterName=TimeSinceStarted}}
 }
 {
  id=ModulatorContinuous_2
  children [WavePlayer_23]
  properties {volumeModulationParams {maxInput=0.25, maxOutput=0, minOutput=1, parameterName=AmmoNormalized}}
 }
 {id=Modulator_0, children [Random_0], properties {pitchMax=1, pitchMin=1}}
 {id=Modulator_1, children [Branch_3], properties {pitchMax=1, pitchMin=1, volumeMax=0.95, volumeMin=0.85}}
 {id=Modulator_3, children [Random_3], properties {pitchMax=1, pitchMin=1, volumeMax=1, volumeMin=0.5}}
 {id=Random_0, children [WavePlayer_1,WavePlayer_2,WavePlayer_3,WavePlayer_4,WavePlayer_8,WavePlayer_9], properties {weights [1,1,1,1,1,1]}}
 {
  id=Random_3
  children [WavePlayer_15, WavePlayer_16, WavePlayer_17, WavePlayer_18, WavePlayer_19, WavePlayer_20]
  properties {weights [1,1,1,1,1,1]}
 }
 {
  id=SoundClass_0
  children [Modulator_1]
  properties {soundClassOverride=/Game/Sounds/FortSoundClassesAndModes/Fort_WeaponRanged_LowAmmo_Critical_PSM}
 }
 {id=SoundClass_1, children [Mixer_1], properties {soundClassOverride=&Fort_WeaponRanged_Fire_1P_PSM}}
 {id=SoundClass_2, children [Modulator_3], properties {soundClassOverride=&Fort_WeaponRanged_Fire_1P_PSM}}
 {id=WavePlayer_1, properties {soundWaveAssetPtr=&/WPN_Precision_AR_Shoot_Close_1p_06}}
 {id=WavePlayer_15, properties {soundWaveAssetPtr=&/WPN_Precision_AR_Shell_Drop_01}}
 {id=WavePlayer_16, properties {soundWaveAssetPtr=&/WPN_Precision_AR_Shell_Drop_02}}
 {id=WavePlayer_17, properties {soundWaveAssetPtr=&/WPN_Precision_AR_Shell_Drop_03}}
 {id=WavePlayer_18, properties {soundWaveAssetPtr=&/WPN_Precision_AR_Shell_Drop_04}}
 {id=WavePlayer_19, properties {soundWaveAssetPtr=&/WPN_Precision_AR_Shell_Drop_05}}
 {id=WavePlayer_2, properties {soundWaveAssetPtr=&/WPN_Precision_AR_Shoot_Close_1p_01}}
 {id=WavePlayer_20, properties {soundWaveAssetPtr=&/WPN_Precision_AR_Shell_Drop_06}}
 {id=WavePlayer_23, properties {soundWaveAssetPtr=&/WPN_Precision_AR_Shoot_Low_Ammo_01}}
 {id=WavePlayer_3, properties {soundWaveAssetPtr=&/WPN_Precision_AR_Shoot_Close_1p_02}}
 {id=WavePlayer_4, properties {soundWaveAssetPtr=&/WPN_Precision_AR_Shoot_Close_1p_03}}
 {id=WavePlayer_8, properties {soundWaveAssetPtr=&/WPN_Precision_AR_Shoot_Close_1p_04}}
 {id=WavePlayer_9, properties {soundWaveAssetPtr=&/WPN_Precision_AR_Shoot_Close_1p_05}}
]
properties {
 attenuationSettings=&Fort_Weapon_Projectile_FirstPerson_Athena_Att
 bDebug=true
 maxDistance=2097152
 soundClassObject=&Fort_WeaponRanged_Fire_1P_PSM
 volumeMultiplier=0.7
 attenuationOverrides {manualReverbSendLevel=-1.0165, reverbSendMethod=Manual}
 concurrencyOverrides {bLimitToOwner=true, maxCount=2, resolutionRule=StopOldest, volumeScale=0.9}
 concurrencySet [&Weapon_Gunfire_FirstPerson_Conc]
}
