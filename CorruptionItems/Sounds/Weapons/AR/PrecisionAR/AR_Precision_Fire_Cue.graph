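# Sound cue for Precision AR gunfire, layering close third-person, medium and distant shot sounds by distance.
# Each layer picks a random wave variation and has its own attenuation and sound class override, with a volume boost on the medium and distant layers.
# Nested mixers blend close with medium, then add distant at a lower level.
# Concurrency sets and an owner-limited stop-oldest override cap overlapping shots.
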SoundCue /CorruptionItems/Sounds/Weapons/AR/PrecisionAR/AR_Precision_Fire_Cue
root=Mixer_2
nodes [
 {id=Attenuation_3, children [Random_1], properties {attenuationSettings=&Fort_Weapon_Projectile_Close_Athena_Att}}
 {
  id=Attenuation_4
  children [Modulator_0]
  properties {attenuationSettings=&Fort_Weapon_Projectile_Distant_Mono_Athena_Att}
 }
 {
  id=Attenuation_5
  children [Modulator_2]
  properties {attenuationSettings=&Fort_Weapon_Projectile_Medium_Mono_Athena_Att}
 }
 {id=Mixer_2, children [Modulator_3,SoundClass_4], properties {inputVolume [1,0.2]}}
 {id=Mixer_3, children [SoundClass_3,SoundClass_5], properties {inputVolume [0.9,0.5]}}
 {id=Modulator_0, children [Random_4], properties {pitchMax=1, pitchMin=1, volumeMax=2, volumeMin=2}}
 {id=Modulator_2, children [Random_3], properties {volumeMax=2, volumeMin=2}}
 {id=Modulator_3, children [Mixer_3], properties {pitchMax=1, pitchMin=1}}
 {
  id=Random_1
  children [WavePlayer_7, WavePlayer_12, WavePlayer_13, WavePlayer_14, WavePlayer_15, WavePlayer_16]
  properties {weights [1,1,1,1,1,1]}
 }
 {id=Random_3, children [WavePlayer_0,WavePlayer_1,WavePlayer_2], properties {weights [1,1,1]}}
 {id=Random_4, children [WavePlayer_8, WavePlayer_9, WavePlayer_10, WavePlayer_11], properties {weights [1,1,1,1]}}
 {
  id=SoundClass_3
  children [Attenuation_3]
  properties {soundClassOverride=&Fort_WeaponRanged_Fire_3P_PSM}
 }
 {
  id=SoundClass_4
  children [Attenuation_4]
  properties {soundClassOverride=&Fort_WeaponRanged_Fire_Distant}
 }
 {
  id=SoundClass_5
  children [Attenuation_5]
  properties {soundClassOverride=&Fort_WeaponRanged_Fire_Distant}
 }
 {id=WavePlayer_0, properties {soundWaveAssetPtr=&/WPN_Precision_AR_Shoot_Med_01}}
 {id=WavePlayer_1, properties {soundWaveAssetPtr=&/WPN_Precision_AR_Shoot_Med_02}}
 {id=WavePlayer_10, properties {soundWaveAssetPtr=&/WPN_Precision_AR_Shoot_Distant_03}}
 {id=WavePlayer_11, properties {soundWaveAssetPtr=&/WPN_Precision_AR_Shoot_Distant_04}}
 {id=WavePlayer_12, properties {soundWaveAssetPtr=&/WPN_Precision_AR_Shoot_Close_3p_02}}
 {id=WavePlayer_13, properties {soundWaveAssetPtr=&/WPN_Precision_AR_Shoot_Close_3p_03}}
 {id=WavePlayer_14, properties {soundWaveAssetPtr=&/WPN_Precision_AR_Shoot_Close_3p_04}}
 {id=WavePlayer_15, properties {soundWaveAssetPtr=&/WPN_Precision_AR_Shoot_Close_3p_05}}
 {id=WavePlayer_16, properties {soundWaveAssetPtr=&/WPN_Precision_AR_Shoot_Close_3p_06}}
 {id=WavePlayer_2, properties {soundWaveAssetPtr=&/WPN_Precision_AR_Shoot_Med_03}}
 {id=WavePlayer_7, properties {soundWaveAssetPtr=&/WPN_Precision_AR_Shoot_Close_3p_01}}
 {id=WavePlayer_8, properties {soundWaveAssetPtr=&/WPN_Precision_AR_Shoot_Distant_01}}
 {id=WavePlayer_9, properties {soundWaveAssetPtr=&/WPN_Precision_AR_Shoot_Distant_02}}
]
properties {
 bDebug=true
 maxDistance=27500
 soundClassObject=&Fort_WeaponRanged_Fire_3P_PSM
 volumeMultiplier=0.9
 concurrencyOverrides {bLimitToOwner=true, maxCount=2, resolutionRule=StopOldest}
 concurrencySet [&Weapon_Gunfire_Global_Conc, &Weapon_Gunfire_Local_Conc]
}
