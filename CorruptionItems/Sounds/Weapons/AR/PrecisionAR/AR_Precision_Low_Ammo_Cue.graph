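# Sound cue for the low-ammo shot sound of the Precision AR in CorruptionItems.
# A single wave player feeds the root directly, with no randomization or modulation, at reduced volume.
# It routes through a ranged-weapon low-ammo sound class and overrides concurrency to prevent new instances rather than cut playing ones.
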SoundCue /CorruptionItems/Sounds/Weapons/AR/PrecisionAR/AR_Precision_Low_Ammo_Cue
root=WavePlayer_0
nodes [{id=WavePlayer_0, properties {soundWaveAssetPtr=&/WPN_Precision_AR_Shoot_Low_Ammo_01}}]
properties {
 bOverrideConcurrency=true
 maxDistance=2097152
 soundClassObject=/Game/Sounds/FortSoundClassesAndModes/Fort_WeaponRanged_LowAmmo_Degraded_PSM
 volumeMultiplier=0.6
 concurrencyOverrides {resolutionRule=PreventNew}
}
