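# Sound cue for the Precision AR's out-of-ammo dry-fire sound.
# A single dry-fire wave plays through a Modulator that slightly randomizes volume while keeping pitch fixed.
# The cue uses the third-person ranged weapon fire sound class, a reduced volume multiplier and an attenuation falloff override.
# Its own concurrency override caps simultaneous instances at a small count, stopping the oldest first.
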SoundCue /CorruptionItems/Sounds/Weapons/AR/PrecisionAR/AR_Precision_OutofAmmo_Cue
root=Modulator_0
nodes [
 {id=Modulator_0, children [WavePlayer_4], properties {pitchMax=1, pitchMin=1, volumeMax=0.95, volumeMin=0.85}}
 {id=WavePlayer_4, properties {soundWaveAssetPtr=&/WPN_Precision_AR_Dryfire_01}}
]
properties {
 bOverrideConcurrency=true
 maxDistance=2097152
 soundClassObject=&Fort_WeaponRanged_Fire_3P_PSM
 volumeMultiplier=0.55
 attenuationOverrides {falloffDistance=1600}
 concurrencyOverrides {maxCount=2, resolutionRule=StopOldest}
}
