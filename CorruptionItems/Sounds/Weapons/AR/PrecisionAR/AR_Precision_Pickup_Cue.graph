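# Sound cue for picking up the Precision AR weapon in the CorruptionItems content.
# A single wave player plays one pickup sound with no randomization, routed through the loot collect sound class and weapon pickup attenuation with a slight volume boost.
# It references the shared PickupMaximum_Oldest concurrency set alongside single-instance, prevent-new override values to limit stacked pickup sounds.
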SoundCue /CorruptionItems/Sounds/Weapons/AR/PrecisionAR/AR_Precision_Pickup_Cue
root=WavePlayer_2
nodes [{id=WavePlayer_2, properties {soundWaveAssetPtr=&/WPN_Precision_AR_Pickup_01}}]
properties {
 attenuationSettings=&Fort_Weapon_Pickup_Att
 maxDistance=3500
 soundClassObject=&Fort_Loot_Collect
 volumeMultiplier=1.3846
 concurrencyOverrides {maxCount=1, resolutionRule=PreventNew}
 concurrencySet [&PickupMaximum_Oldest]
}
