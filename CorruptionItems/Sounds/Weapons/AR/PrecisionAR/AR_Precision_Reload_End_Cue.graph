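# Sound cue for the end stage of the Precision AR reload in the weapon sound set.
# A single wave player feeds the reload-end wave straight to the output, with no randomization or mixing.
# It uses the shared weapon reload attenuation and weapon foley sound class, with reduced volume and a concurrency override, and has its debug flag on.
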SoundCue /CorruptionItems/Sounds/Weapons/AR/PrecisionAR/AR_Precision_Reload_End_Cue
root=WavePlayer_0
nodes [{id=WavePlayer_0, properties {soundWaveAssetPtr=&/WPN_Precision_AR_Reload_End}}]
properties {
 attenuationSettings=&Fort_Weapon_Reload_Att
 bDebug=true
 bOverrideConcurrency=true
 maxDistance=2250
 soundClassObject=&Fort_WeaponFoley
 volumeMultiplier=0.752
}
